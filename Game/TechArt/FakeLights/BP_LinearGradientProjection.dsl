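; Fake light actor that projects a linear gradient through a material on a cube sized from projection width, height and depth.
; ApplyLightSettings, run from construction and BeginPlay, writes color, falloff, intensity and projection settings into the cube's custom primitive data.
; Use TOD swaps to a time-of-day material fed morning, day, evening and night colors.
; World Z scale is injected as primitive data to spare ObjectScale shader instructions.
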(event EventBeginPlay
  (CallFunction|ApplyLightSettings))

(event Collision|EventActorBeginOverlap
  (params (OtherActor :type Actor%20Object%20Reference)))

(event EventTick
  (params (DeltaSeconds :type Float%20%28single-precision%29)))

(fn ApplyLightSettings
  (Utilities|FlowControl|Sequence
    (:then_0
      ; Light color with TOD
      (if (Variables|3-TImeOfDay|GetUse%20TOD)
        (Rendering|Material|SetMaterial (Variables|Default|GetCube) 0 "/Game/TechArt/FakeLights/Materials/MI_LinearGradientProjection_TOD.MI_LinearGradientProjection_TOD")
        (Rendering|Material|SetDefaultCustomPrimitiveDataVector3 (Variables|Default|GetCube) 17 (Math|Conversions|ToVector(LinearColor) (Math|Conversions|ToLinearColor(Color) (Variables|3-TImeOfDay|GetColor%20Morning))))
        (Rendering|Material|SetDefaultCustomPrimitiveDataVector3 (Variables|Default|GetCube) 20 (Math|Conversions|ToVector(LinearColor) (Math|Conversions|ToLinearColor(Color) (Variables|3-TImeOfDay|GetColor%20Day))))
        (Rendering|Material|SetDefaultCustomPrimitiveDataVector3 (Variables|Default|GetCube) 23 (Math|Conversions|ToVector(LinearColor) (Math|Conversions|ToLinearColor(Color) (Variables|3-TImeOfDay|GetColor%20Evening))))
        (Rendering|Material|SetDefaultCustomPrimitiveDataVector3 (Variables|Default|GetCube) 26 (Math|Conversions|ToVector(LinearColor) (Math|Conversions|ToLinearColor(Color) (Variables|3-TImeOfDay|GetColor%20Night))))
        (else
          (Rendering|Material|SetMaterial (Variables|Default|GetCube) 0 "/Game/TechArt/FakeLights/Materials/MI_LinearGradientProjection.MI_LinearGradientProjection")
          (Rendering|Material|SetDefaultCustomPrimitiveDataVector3 (Variables|Default|GetCube) 0 (Math|Conversions|ToVector(LinearColor) (Math|Conversions|ToLinearColor(Color) (Variables|1-LightSettings|GetLight%20Color)))))))
    (:then_1
      ; MAsk
      (Rendering|Material|SetDefaultCustomPrimitiveDataFloat (Variables|Default|GetCube) 9 (Variables|1-LightSettings|GetLight%20Falloff%20U))
      (Rendering|Material|SetDefaultCustomPrimitiveDataFloat (Variables|Default|GetCube) 10 (Variables|1-LightSettings|GetLight%20Falloff%20V))
      (Rendering|Material|SetDefaultCustomPrimitiveDataFloat (Variables|Default|GetCube) 11 (Variables|2-ProjectionSettings|GetProjection%20Depth))
      (Rendering|Material|SetDefaultCustomPrimitiveDataFloat (Variables|Default|GetCube) 12 (Variables|2-ProjectionSettings|GetProjection%20Offset%20U))
      (Rendering|Material|SetDefaultCustomPrimitiveDataFloat (Variables|Default|GetCube) 13 (Variables|2-ProjectionSettings|GetProjection%20Offset%20V))
      (Rendering|Material|SetDefaultCustomPrimitiveDataFloat (Variables|Default|GetCube) 14 (Variables|1-LightSettings|GetLight%20Intensity))
      (Rendering|Material|SetDefaultCustomPrimitiveDataFloat (Variables|Default|GetCube) 15 (Variables|2-ProjectionSettings|GetProjection%20Clipping))
      (Rendering|Material|SetDefaultCustomPrimitiveDataFloat (Variables|Default|GetCube) 16 (Math|Conversions|ToFloat(Integer) (Variables|2-ProjectionSettings|GetRepeat%20On%20U)))
      (Rendering|Material|SetDefaultCustomPrimitiveDataFloat (Variables|Default|GetCube) 30 (Variables|2-ProjectionSettings|GetProjection%20Edge%20Falloff)))
    (:then_2
      ; Scale cube based on projection parameters
      (Transformation|SetRelativeScale3D (Variables|Default|GetCube) (Math|Vector|MakeVector (/ (Variables|2-ProjectionSettings|GetProjection%20Width) 1e+02) (/ (Variables|2-ProjectionSettings|GetProjection%20Height) 1e+02) (/ (Variables|2-ProjectionSettings|GetProjection%20Depth) 1e+02))))
    (:then_3
      ; Inject world scale to avoid using ObjectScale in shader - saves 20 inst
      (Rendering|Material|SetDefaultCustomPrimitiveDataFloat (Variables|Default|GetCube) 3 (.z (Transformation|GetWorldScale (Variables|Default|GetCube)))))))

(fn ConstructionScript
  (CallFunction|ApplyLightSettings))
