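; Fake point light Blueprint that feeds per-instance look settings to the material on its SM_PointDecal component.
; SetParameters packs morning, day, evening and night colors, intensity, falloff, positive and negative bounds with falloff, additive amount and near fade range into fixed Custom Primitive Data slots.
; It runs from the Construction Script and again on BeginPlay, writing default primitive data without creating dynamic material instances.
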(event EventBeginPlay
  (CallFunction|SetParameters))

(fn SetParameters
  (Rendering|Material|SetDefaultCustomPrimitiveDataVector3 (Variables|Default|GetSM_PointDecal) 0 (Math|Conversions|ToVector(LinearColor) (Variables|Default|GetColor%20Morning)))
  (Rendering|Material|SetDefaultCustomPrimitiveDataVector3 (Variables|Default|GetSM_PointDecal) 3 (Math|Conversions|ToVector(LinearColor) (Variables|Default|GetColor%20Day)))
  (Rendering|Material|SetDefaultCustomPrimitiveDataVector3 (Variables|Default|GetSM_PointDecal) 6 (Math|Conversions|ToVector(LinearColor) (Variables|Default|GetColor%20Evening)))
  (Rendering|Material|SetDefaultCustomPrimitiveDataVector3 (Variables|Default|GetSM_PointDecal) 9 (Math|Conversions|ToVector(LinearColor) (Variables|Default|GetColor%20Night)))
  (Rendering|Material|SetDefaultCustomPrimitiveDataFloat (Variables|Default|GetSM_PointDecal) 12 (Variables|Default|GetIntensity))
  (Rendering|Material|SetDefaultCustomPrimitiveDataFloat (Variables|Default|GetSM_PointDecal) 13 (Variables|Default|GetFalloff))
  (Rendering|Material|SetDefaultCustomPrimitiveDataVector3 (Variables|Default|GetSM_PointDecal) 14 (Variables|Default|GetBounds%20+))
  (Rendering|Material|SetDefaultCustomPrimitiveDataVector3 (Variables|Default|GetSM_PointDecal) 17 (Variables|Default|GetBounds%20-))
  (Rendering|Material|SetDefaultCustomPrimitiveDataFloat (Variables|Default|GetSM_PointDecal) 20 (Variables|Default|GetBounds%20Falloff))
  (Rendering|Material|SetDefaultCustomPrimitiveDataFloat (Variables|Default|GetSM_PointDecal) 21 (Variables|Default|GetAdditive))
  (Rendering|Material|SetDefaultCustomPrimitiveDataFloat (Variables|Default|GetSM_PointDecal) 22 (Variables|Default|GetNear%20Fade%20End))
  (Rendering|Material|SetDefaultCustomPrimitiveDataFloat (Variables|Default|GetSM_PointDecal) 23 (Variables|Default|GetNear%20Fade%20Start)))

(fn ConstructionScript
  (CallFunction|SetParameters))
