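# Material function that computes a spotlight-style cone mask for fake lights from local-space scene position behind a box and world-space scene normals.
# Box-normalized coordinates, with user offset and U-axis repeat, feed cone angle and distance falloff terms shaped by custom primitive data parameters.
# The mask sums the cone, an NdotL-weighted term (constant on ES2/ES3_1, where normals can't be sampled) and a GGX specular lobe gated by platform switch.
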MaterialFunction /Game/TechArt/FakeLights/Materials/MF_LightConeMask
graph {nodes [
  n1 = Subtract(A=NamedRerouteUsage(), B=NamedRerouteUsage())
  n2 = NamedRerouteUsage()
  n3 = Multiply(A=Saturate(None=Divide(A=@n1, B=@n2)), B=Step(X=@n1))
  n4 = Saturate(None=Multiply(A=OneMinus(None=@n3), B=NamedRerouteUsage())) {desc=Inner Cone}
  n5 = NamedRerouteUsage()
  n6 = Saturate(None=Multiply(A=Add(A=@n4, B=@n5), B=NamedRerouteUsage()))
  n7 = NamedRerouteUsage()
  n8 = Multiply(A=Transform(None=Normalize(VectorInput=@n7)) {transformSourceType=TRANSFORMSOURCE_Local}) {constB=-1, desc=Light direction}
  n9 = FunctionInput() {inputName=SceneNormals_WS, sortPriority=2}
  n10 = Constant() {r=1}
  n11 = Multiply(
   A=@n6
   B=FeatureLevelSwitch(Default=DotProduct(A=@n8, B=@n9) {desc=NdotL}, ES2=@n10, ES3_1=@n10) {desc=Normal Buffer can't be sampled on ES3_1}
  ) {constA=1}
  n12 = Constant()
  n13 = Multiply(
   A=@n11
   B=MaterialFunctionCall!(Normal=ReflectionVectorWS(CustomWorldNormal=@n9), Light Vector=@n8, Roughness=Constant() {r=0.5}) {materialFunction=/Engine/Functions/Engine_MaterialFunctions01/Lighting/GGXSpecular}.Result
  )
  FunctionOutput(None=Add(
    A=Add(A=@n6, B=@n11)
    B=MaterialFunctionCall("Xb1,PS4"=@n12, "PC-High&&Epic,XSX,PS5"=@n13, "PC-Perf,Android"=@n12, Switch=@n12, "PC-Low(NotPerf)&&Medium"=@n12) {materialFunction=&PlatformSwitch_1D}.Result
   ) {constB=0, desc=Spec lobe}) {outputName=Mask}
  n14 = NamedRerouteDeclaration(None=Multiply(A=FunctionInput() {inputName=PositionBehind_LS, sortPriority=1}) {constB=0.01, desc=Normalized based on box size}) {name=PositionBehind_Normalized, nodeColor=<rgba 0.4531 0 1 1>}
  n15 = Constant3Vector() {constant=<rgba 0.5 0.5 0 0>}
  MaterialExpressionAdd_10 = Add(
   A=NamedRerouteDeclaration(None=Add(A=@n14, B=@n15) {constB=-0.5}) {name=UVWs_NoUserOffset, nodeColor=<rgba 0.7812 1 0 1>}
   B=AppendVector(A=FunctionInput() {inputName=UVOffset, inputType=FunctionInput_Vector2}, B=Constant())
  ) {constB=0, desc=User Offset}
  n16 = Convert(None=@MaterialExpressionAdd_10) {
   nodeName=Break Float3
   convertInputs [{
     type=Vector3
     expressionInput {
      expression=&self:MaterialExpressionAdd_10
      inputName=None
      mask=0
      maskA=0
      maskB=0
      maskG=0
      maskR=0
      outputIndex=0
     }
    }]
   convertOutputs [{type=Scalar}, {type=Scalar}, {type=Scalar}]
  }
  n17 = Multiply(A=@n16, B=ScalarParameter() {bUseCustomPrimitiveData=true, defaultValue=1, parameterName=Repeat on U, primitiveDataIndex=16})
  n18 = AppendVector(A=Subtract(A=Frac(None=@n17)) {constB=0.5}, B=@n16)
  NamedRerouteDeclaration(None=AppendVector(A=@n18, B=@n16)) {name=UVWs, nodeColor=<rgba 0 1 0.9062 1>}
  n19 = OneMinus(None=Length(None=NamedRerouteUsage()) {is=class MaterialExpressionLength})
  n20 = Saturate(None=Divide(A=@n19, B=NamedRerouteUsage()) {constB=1})
  n21 = Power(Base=@n20, Exp=NamedRerouteUsage()) {constExponent=4.5}
  n22 = SmoothStep(Value=@n19) {constMax=0.2}
  n23 = NamedRerouteUsage()
  n24 = OneMinus(None=Saturate(None=ComponentMask(None=@n23) {g=true}))
  NamedRerouteDeclaration(None=Multiply(A=LinearInterpolate(A=@n20, B=@n21, Alpha=@n22), B=Multiply(A=@n24, B=@n24) {desc="Distance^2"})) {name=Cone Falloff, nodeColor=<rgba 0 1 0.1328 1>}
  NamedRerouteDeclaration(None=Constant3Vector() {constant=<rgba 0 1 0 0>}) {name=Light Dir, nodeColor=<rgba 0 1 0.4609 1>}
  n25 = AppendVector(A=NamedRerouteUsage(), B=NamedRerouteUsage())
  NamedRerouteDeclaration(None=DotProduct(A=Normalize(VectorInput=@n25), B=NamedRerouteUsage())) {name=CosTheta, nodeColor=<rgba 0 1 0.6953 1>}
  n26 = ScalarParameter() {
   bUseCustomPrimitiveData=true
   defaultValue=15
   group=0-Light
   parameterName=Inner Cone Angle
   primitiveDataIndex=4
   sliderMax=85
   sliderMin=1
   sortPriority=2
  }
  n27 = Cosine(None=Multiply(A=Round(None=@n26), B=Divide() {constB=360}) {constB=0.0028})
  n28 = ScalarParameter() {
   bUseCustomPrimitiveData=true
   defaultValue=50
   group=0-Light
   parameterName=Outer Cone Angle
   primitiveDataIndex=6
   sliderMax=90
   sliderMin=1
   sortPriority=3
  }
  n29 = Multiply(A=Round(None=@n28), B=Divide() {constB=360}) {constB=0.0028, desc=Degrees to turns}
  NamedRerouteDeclaration(None=Min(A=NamedRerouteDeclaration(None=@n27) {name=Cos Inner, nodeColor=<rgba 0.6641 0 1 1>}, B=Cosine(None=@n29))) {name=Cos Outer, nodeColor=<rgba 1 0.9844 0 1>}
  n30 = NamedRerouteUsage()
  n31 = NamedRerouteUsage()
  n32 = NamedRerouteUsage()
  NamedRerouteDeclaration(None=Saturate(None=SmoothStep(Min=@n30, Max=@n31, Value=@n32))) {name=Cone Base, nodeColor=<rgba 0 1 0.2969 1>}
  n33 = ScalarParameter() {
   bUseCustomPrimitiveData=true
   defaultValue=0.5
   group=0-Light
   parameterName=Inner Cone Softness
   primitiveDataIndex=7
   sliderMax=1
   sortPriority=5
  }
  n34 = Constant2Vector() {g=0.1, r=0}
  NamedRerouteDeclaration(None=MaterialFunctionCall(Input=Saturate(None=@n33), Target Low=@n34.R, Target High=@n34.R) {materialFunction=&RemapValueRange}.Result) {name=Inner Cone Delta, nodeColor=<rgba 1 0.6797 0 1>}
  n35 = ScalarParameter() {
   bUseCustomPrimitiveData=true
   defaultValue=0.2
   group=0-Light
   parameterName=Cone Origin Softness
   primitiveDataIndex=8
   sliderMax=1
   sortPriority=6
  }
  n36 = Constant2Vector() {g=0.1}
  NamedRerouteDeclaration(None=MaterialFunctionCall(Input=Saturate(None=@n35), Target Low=@n36.R, Target High=@n36.R) {materialFunction=&RemapValueRange}.Result) {name=Cone Depth, nodeColor=<rgba 0 1 0.4141 1>}
  n37 = ScalarParameter() {
   bUseCustomPrimitiveData=true
   defaultValue=0.5
   group=0-Light
   parameterName=Inner Cone Weight
   primitiveDataIndex=5
   sliderMax=1
   sortPriority=4
  }
  n38 = Constant2Vector() {g=0.25}
  NamedRerouteDeclaration(None=MaterialFunctionCall(Input=Saturate(None=@n37), Target Low=@n38.R, Target High=@n38.R) {materialFunction=&RemapValueRange}.Result) {name=Inner Cone Weight, nodeColor=<rgba 1 0.8672 0 1>}
  NamedRerouteDeclaration(None=@n18) {name=UVs, nodeColor=<rgba 0.0312 1 0 1>}
  n39 = ScalarParameter() {
   bUseCustomPrimitiveData=true
   defaultValue=0.25
   group=0-Light
   parameterName=Hotspot Intensity
   primitiveDataIndex=9
   sliderMax=1
   sortPriority=7
  }
  n40 = Constant2Vector() {g=0.9, r=1}
  NamedRerouteDeclaration(None=MaterialFunctionCall(Input=Saturate(None=@n39), Target Low=@n40.R, Target High=@n40.R) {materialFunction=&RemapValueRange}.Result) {name=Hotspot, nodeColor=<rgba 0 1 0.6484 1>}
  n41 = ScalarParameter() {
   bUseCustomPrimitiveData=true
   defaultValue=2
   group=0-Light
   parameterName=Light Falloff Exponent
   primitiveDataIndex=10
   sliderMax=50
   sortPriority=8
  }
  NamedRerouteDeclaration(None=Multiply(A=@n41) {constB=0.3828, desc=Magic number to align with spotlight}) {name=Light Falloff Exponent, nodeColor=<rgba 1 0.6094 0 1>}
 ]}
