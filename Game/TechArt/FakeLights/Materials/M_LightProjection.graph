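# Additive unlit fake-light material that projects cone or linear-gradient light onto scene surfaces inside a box volume, supporting instanced meshes.
# Depth-reconstructed local positions and scene normals drive masks, edge falloff and clipping, tinted by base color plus slight reflectance and compensated for eye adaptation.
# World position offset handles a camera inside the box and apparently collapses black lights.
# Custom primitive data drives most parameters, with optional time-of-day color and ES3_1 normal fallbacks.
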Material /Game/TechArt/FakeLights/Materials/M_LightProjection
graph {
 properties {
  bAutomaticallySetUsageInEditor=false
  bUsedWithInstancedStaticMeshes=true
  blendMode=BLEND_Additive
  shadingModel=MSM_Unlit
  usedShadingModels=Unlit
 }
 nodes [
  n1 = NamedRerouteUsage()
  n2 = LinearInterpolate(A=Multiply(A=NamedRerouteUsage(), B=@n1), B=@n1, Alpha=Multiply() {constA=0.5, constB=0.08, desc=4% reflectance}) {desc=Absorbed / Reflected}
  n3 = NamedRerouteUsage()
  n4 = NamedRerouteUsage()
  n5 = NamedRerouteUsage()
  n6 = StaticSwitchParameter(
   True=MaterialFunctionCall!(UVOffset=@n3, PositionBehind_LS=@n4, SceneNormals_WS=@n5) {materialFunction=&/MF_LinearGradientMask}.Mask
   False=StaticSwitchParameter(
    True=MaterialFunctionCall!(UVOffset=@n3, PositionBehind_LS=@n4, SceneNormals_WS=@n5) {materialFunction=&/MF_LinearConeMask}.Mask
    False=MaterialFunctionCall!(UVOffset=@n3, PositionBehind_LS=@n4, SceneNormals_WS=@n5) {materialFunction=&/MF_LightConeMask}.Mask
   ) {group=0-Light, parameterName=UseLinearCone, sortPriority=-1}
  ) {defaultValue=true, group=0-Light, parameterName=Use Linear Gradient, sortPriority=-2}
  n7 = Multiply(A=Multiply(A=Multiply(A=@n2, B=@n6), B=NamedRerouteUsage()), B=NamedRerouteUsage())
  n8 = ScalarParameter() {bUseCustomPrimitiveData=true, defaultValue=1, group=0-Light, parameterName=Light Intensity, primitiveDataIndex=14, sortPriority=1}
  Divide(A=Multiply(A=Saturate(None=@n7), B=@n8), B=EyeAdaptation().EyeAdaptation)
  n9 = Multiply(A=Divide(A=SceneDepth().R, B=PixelDepth().R), B=WorldPosition() {worldPositionShaderOffset=WPT_CameraRelative}.XYZ)
  n10 = ActorPositionWS() {originType=CameraRelative}
  NamedRerouteDeclaration(None=Transform(None=Subtract(A=@n9, B=@n10)) {transformSourceType=TRANSFORMSOURCE_World, transformType=TRANSFORM_Local}) {name=PositionBehind LS, nodeColor=<rgba 0 1 0.25 1>}
  NamedRerouteDeclaration(None=ComponentMask(None=SceneTexture() {sceneTextureId=PPI_StoredBaseColor}.Color) {b=true, g=true, r=true}) {name=Scene BaseColor, nodeColor=<rgba 0.1016 1 0 1>}
  n11 = ComponentMask(None=Transform(None=NamedRerouteUsage()) {transformSourceType=TRANSFORMSOURCE_World, transformType=TRANSFORM_Local}) {b=true}
  n12 = ScalarParameter() {bUseCustomPrimitiveData=true, defaultValue=1, parameterName=WorldScaleZ, primitiveDataIndex=3}
  n13 = ScalarParameter() {bUseCustomPrimitiveData=true, parameterName=Clipping Threshold, primitiveDataIndex=15}
  n14 = Subtract(A=Abs(None=Multiply(A=@n11, B=@n12)), B=Subtract(A=Multiply(A=@n13) {constB=2})) {constB=0, desc="[-1,1]"}
  n15 = Constant() {r=1}
  NamedRerouteDeclaration(None=FeatureLevelSwitch(Default=Saturate(None=@n14), ES2=@n15, ES3_1=@n15) {desc=Normal Buffer can't be sampled on ES3_1}) {name=NormalClipping, nodeColor=<rgba 1 0.1875 0 1>}
  n16 = Subtract(
   A=TransformPosition(None=Constant3Vector()) {transformSourceType=TRANSFORMPOSSOURCE_TranslatedWorld}
   B=Constant3Vector() {constant=<rgba 0 50 50 0>}
  ) {desc=Pivot back to center of box}
  MaterialExpressionMultiply_16 = Multiply(A=Subtract(A=Abs(None=@n16)) {desc="Offset to mitigate\r\n pop between\r\n interior and exterior"}) {constB=0.02, desc="half extend of a box\r\n 100x100x100"}
  n17 = Convert(None=@MaterialExpressionMultiply_16) {
   nodeName=Break Float3
   convertInputs [{
     type=Vector3
     expressionInput {
      expression=&self:MaterialExpressionMultiply_16
      inputName=None
      mask=0
      maskA=0
      maskB=0
      maskG=0
      maskR=0
      outputIndex=0
     }
    }]
   convertOutputs [{type=Scalar}, {type=Scalar}, {type=Scalar}]
  }
  n18 = Floor(None=Max(A=Max(A=@n17, B=@n17), B=@n17))
  NamedRerouteDeclaration(None=OneMinus(None=Saturate(None=@n18))) {name=is Inside Box, nodeColor=<rgba 1 0.5625 0 1>}
  n19 = WorldPosition() {worldPositionShaderOffset=WPT_CameraRelative}
  n20 = Add(
   A=Multiply(A=LocalPosition().XYZ, B=Constant3Vector() {constant=<rgba 1 -1 1 0>})
   B=Constant3Vector() {desc=Offset infront of camera, constant=<rgba 0 0 20 0>}
  ) {constB=0}
  n21 = LinearInterpolate(
   A=Multiply(A=VertexColor().R, B=Multiply(A=@n19.XYZ) {constB=-1}) {constA=1}
   B=Subtract(A=Transform(None=@n20) {transformSourceType=TRANSFORMSOURCE_View}, B=WorldPosition() {worldPositionShaderOffset=WPT_CameraRelative}.XYZ)
   Alpha=NamedRerouteUsage()
  ) {desc="0: Outside\r\n1: Inside"}
  n22 = VertexColor()
  n23 = WorldPosition() {worldPositionShaderOffset=WPT_CameraRelative}
  NamedRerouteDeclaration(None=LinearInterpolate(A=Multiply(A=@n21, B=@n22.R), B=Multiply(A=@n23.XYZ) {constB=-1}, Alpha=NamedRerouteUsage()) {constB=0}) {name=WPO, nodeColor=<rgba 0.3359 1 0 1>}
  n24 = ScalarParameter() {bUseCustomPrimitiveData=true, group=1-Projection, parameterName=Projection Offset U, primitiveDataIndex=12, sortPriority=21}
  n25 = ScalarParameter() {bUseCustomPrimitiveData=true, group=1-Projection, parameterName=Projection Offset V, primitiveDataIndex=13, sortPriority=22}
  NamedRerouteDeclaration(None=MaterialFunctionCall(X=@n24, Y=@n25) {materialFunction=&MakeFloat2}.Result) {name=UV_Offset, nodeColor=<rgba 0 1 0.7656 1>}
  n26 = ComponentMask(None=SceneTexture() {sceneTextureId=PPI_WorldNormal}.Color) {b=true, g=true, r=true}
  n27 = VertexNormalWS()
  NamedRerouteDeclaration(None=Normalize(VectorInput=FeatureLevelSwitch(Default=@n26, ES2=@n27, ES3_1=@n27) {desc=Normal Buffer can't be sampled on ES3_1})) {name=Scene Normals WS, nodeColor=<rgba 0 0.7422 1 1>}
  n28 = VectorParameter() {bUseCustomPrimitiveData=true, parameterName=Color Morning, primitiveDataIndex=17, defaultValue=<rgba 0.7518 0.0417 1 1>}
  n29 = VectorParameter() {bUseCustomPrimitiveData=true, parameterName=Color Day, primitiveDataIndex=20, defaultValue=<rgba 0.6726 0.2965 1 1>}
  n30 = VectorParameter() {bUseCustomPrimitiveData=true, parameterName=Color Evening, primitiveDataIndex=23, defaultValue=<rgba 0.7518 0.0417 1 1>}
  n31 = VectorParameter() {bUseCustomPrimitiveData=true, parameterName=Color Night, primitiveDataIndex=26, defaultValue=<rgba 0.7518 0.0417 1 1>}
  n32 = VectorParameter() {bUseCustomPrimitiveData=true, group=0-Light, parameterName=Light Color, sortPriority=0, defaultValue=<rgba 1 1 1 1>}
  NamedRerouteDeclaration(None=StaticSwitchParameter(
    True=MaterialFunctionCall(Color_Morning=@n28.RGB, Color_Day=@n29.RGB, Color_Evening=@n30.RGB, Color_Night=@n31.RGB) {materialFunction=&Color_Time_of_Day}.Result
    False=@n32.RGB
   ) {parameterName=Use TOD}) {name=LightColor, nodeColor=<rgba 1 0.7969 0 1>}
  n33 = NamedRerouteUsage()
  n34 = Constant() {r=1}
  n35 = Ceil(None=Add(A=DotProduct(A=@n33, B=@n34) {desc=Sum components}) {constB=-0.01})
  NamedRerouteDeclaration(None=OneMinus(None=Saturate(None=@n35))) {name=is light Black, nodeColor=<rgba 1 0.9375 0 1>}
  n36 = NamedRerouteUsage()
  n37 = Constant3Vector() {constant=<rgba 0 50 50 0>}
  n38 = Abs(None=Multiply(A=Subtract(A=@n36, B=@n37) {desc="Pivot back to center\r\n of box"}) {constB=0.02, desc="half extend of a box\r\n 100x100x100"})
  MaterialExpressionSmoothStep_7 = SmoothStep(Value=Multiply(A=@n38, B=@n38))
  n39 = Convert(None=@MaterialExpressionSmoothStep_7) {
   nodeName=Break Float3
   convertInputs [{
     type=Vector3
     expressionInput {
      expression=&self:MaterialExpressionSmoothStep_7
      inputName=None
      mask=0
      maskA=0
      maskB=0
      maskG=0
      maskR=0
      outputIndex=0
     }
    }]
   convertOutputs [{type=Scalar}, {type=Scalar}, {type=Scalar}]
  }
  n40 = Max(A=Max(A=@n39, B=@n39), B=@n39)
  NamedRerouteDeclaration(None=Power(Base=OneMinus(None=@n40), Exp=ScalarParameter() {bUseCustomPrimitiveData=true, parameterName=Edge Falloff, primitiveDataIndex=30})) {name=Projection Bounds, nodeColor=<rgba 0 0.5078 1 1>}
  NamedRerouteUsage()
 ]
}
