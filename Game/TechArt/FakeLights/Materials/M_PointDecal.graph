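# Additive unlit fake point-light material that lights scene surfaces behind its mesh, using positions rebuilt from scene depth.
# GBuffer normals and base color combine with radial falloff, soft box-bounds masks, distance fade and time-of-day color, mostly driven by Custom Primitive Data.
# World position offset logic clamps vertices to the bounds, blends by red vertex color toward view-space placement, and collapses the mesh as a culling trick.
# Camera inside/outside volume tests supply culling masks.
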Material /Game/TechArt/FakeLights/Materials/M_PointDecal
graph {
 properties {bAutomaticallySetUsageInEditor=false, blendMode=BLEND_Additive, shadingModel=MSM_Unlit, usedShadingModels=Unlit}
 nodes [
  n1 = ScalarParameter() {bUseCustomPrimitiveData=true, defaultValue=15000, group=NearFade, parameterName=Near Fade Start, primitiveDataIndex=23}
  n2 = ActorPositionWS() {originType=CameraRelative}
  n3 = Constant()
  n4 = Saturate(None=InverseLinearInterpolate(
    A=ScalarParameter() {bUseCustomPrimitiveData=true, defaultValue=7500, group=NearFade, parameterName=Near Fade End, primitiveDataIndex=22}
    B=@n1
    Value=Distance(A=@n2, B=@n3)
   ) {constA=2000, constB=5000})
  n5 = NamedRerouteUsage()
  Max(A=@n4, B=@n5)
  n6 = Transform(None=Subtract(A=WorldPosition() {worldPositionShaderOffset=WPT_CameraRelative}.XYZ, B=ActorPositionWS() {originType=CameraRelative})) {transformSourceType=TRANSFORMSOURCE_World, transformType=TRANSFORM_Local}
  n7 = VectorParameter() {bUseCustomPrimitiveData=true, parameterName=Bounds +, primitiveDataIndex=14, defaultValue=<rgba 1 1 1 1>}
  n8 = NamedRerouteDeclaration(None=Multiply(A=@n7.RGB) {constB=100}) {name=Bounds +, nodeColor=<rgba 0 1 0.6953 1>}
  n9 = VectorParameter() {bUseCustomPrimitiveData=true, parameterName=Bounds -, primitiveDataIndex=17, defaultValue=<rgba 1 1 1 1>}
  n10 = NamedRerouteDeclaration(None=Multiply(A=@n9.RGB) {constB=100}) {name=Bounds -, nodeColor=<rgba 1 0.375 0 1>}
  n11 = Add(A=Max(A=Subtract(A=@n6, B=@n8)) {constB=0}, B=Min(A=Add(A=@n6, B=@n10)) {constB=0})
  n12 = LinearInterpolate(A=NamedRerouteUsage(), B=Transform(None=Multiply(A=@n11) {constB=-1}) {transformSourceType=TRANSFORMSOURCE_Local}, Alpha=NamedRerouteUsage()) {constAlpha=1}
  n13 = Add(A=Multiply(A=LocalPosition().XYZ, B=Constant3Vector() {constant=<rgba 1 -1 1 0>}), B=Constant3Vector() {constant=<rgba 0 0 20 0>}) {constB=-100}
  n14 = Add(A=NamedRerouteUsage(), B=Multiply(A=Transform(None=@n13) {transformSourceType=TRANSFORMSOURCE_View}, B=NamedRerouteUsage()))
  n15 = VertexColor()
  n16 = WorldPosition() {worldPositionShaderOffset=WPT_CameraRelative}
  NamedRerouteDeclaration(None=If(A=Min(A=@n4, B=@n5), "A > B"=LinearInterpolate(A=@n12, B=@n14, Alpha=@n15.R), "A < B"=Multiply(A=@n16.XYZ) {constB=-1}) {constB=0.01}) {name=WPO, nodeColor=<rgba 0.4766 1 0 1>}
  n17 = Multiply(A=NamedRerouteUsage()) {constB=-1}
  n18 = Multiply(A=DotProduct(A=SceneTexture() {sceneTextureId=PPI_WorldNormal}.Color, B=Normalize(VectorInput=@n17)), B=NamedRerouteUsage())
  n19 = NamedRerouteUsage()
  n20 = NamedRerouteUsage()
  n21 = Multiply(
   A=Multiply(A=Multiply(A=@n18, B=@n19), B=VertexInterpolator(VS=@n20).PS)
   B=ScalarParameter() {bUseCustomPrimitiveData=true, defaultValue=1, parameterName=Intensity, primitiveDataIndex=12}
  )
  n22 = Multiply(A=@n21, B=ComponentMask(None=SceneTexture() {sceneTextureId=PPI_StoredBaseColor}.Color) {b=true, g=true, r=true})
  n23 = Multiply(A=@n21, B=ScalarParameter() {bUseCustomPrimitiveData=true, defaultValue=1, parameterName=Additive, primitiveDataIndex=21}) {constB=0.1}
  NamedRerouteDeclaration(None=Multiply(A=Add(A=@n22, B=@n23), B=NamedRerouteUsage())) {name=Final Color, nodeColor=<rgba 0.2422 0 1 1>}
  NamedRerouteDeclaration(None=VertexInterpolator(VS=@n4).PS) {name=Distance Fade, nodeColor=<rgba 0 1 0.0625 1>}
  n24 = SceneDepth()
  n25 = PixelDepth()
  n26 = Subtract(
   A=Multiply(A=Divide(A=@n24.R, B=@n25.R), B=WorldPosition() {worldPositionShaderOffset=WPT_CameraRelative}.XYZ)
   B=ActorPositionWS() {originType=CameraRelative}
  )
  NamedRerouteDeclaration(None=Transform(None=NamedRerouteDeclaration(None=@n26) {name=WorldPosBehindTranslucency, nodeColor=<rgba 0.2422 1 0 1>}) {transformSourceType=TRANSFORMSOURCE_World, transformType=TRANSFORM_Local}) {name=WorldPosBehindTranslucencyLocal, nodeColor=<rgba 0 0.1217 1 1>}
  n27 = LocalPosition() {includedOffsets=ExcludeOffsets}
  NamedRerouteDeclaration(None=Frac(None=Divide(A=@n27.XYZ))) {name=Debug, nodeColor=<rgba 0 1 0.7422 1>}
  n28 = Multiply(A=WorldPosition() {worldPositionShaderOffset=WPT_CameraRelative}.XYZ) {constB=-1}
  NamedRerouteDeclaration(None=@n28) {name=Collapse, nodeColor=<rgba 0.6875 0 1 1>}
  n29 = Multiply(A=Multiply(A=NamedRerouteUsage()) {constB=0.01}) {constB=1.15}
  MaterialExpressionMultiply_3 = Multiply(A=@n29, B=@n29)
  n30 = Convert(None=@MaterialExpressionMultiply_3) {
   nodeName=Break Float3
   convertInputs [{
     type=Vector3
     expressionInput {
      expression=&self:MaterialExpressionMultiply_3
      inputName=None
      mask=0
      maskA=0
      maskB=0
      maskG=0
      maskR=0
      outputIndex=0
     }
    }]
   convertOutputs [{type=Scalar}, {type=Scalar}, {type=Scalar}]
  }
  n31 = Saturate(None=Add(A=Add(A=@n30, B=@n30), B=@n30))
  NamedRerouteDeclaration(None=Power(Base=OneMinus(None=@n31), Exp=ScalarParameter() {bUseCustomPrimitiveData=true, defaultValue=2, parameterName=Falloff, primitiveDataIndex=13})) {name=Radial Mask, nodeColor=<rgba 1 0 0.5625 1>}
  n32 = NamedRerouteUsage()
  n33 = Subtract(A=@n32, B=VertexInterpolator(VS=NamedRerouteUsage()).PS)
  n34 = VertexInterpolator(VS=MaterialFunctionCall() {materialFunction=&ObjectScale}.Scale XYZ)
  n35 = ScalarParameter() {bUseCustomPrimitiveData=true, defaultValue=1, parameterName=Bounds Fade, primitiveDataIndex=20}
  n36 = Max(A=Add(A=Multiply(A=@n33, B=@n34.PS), B=@n35) {constB=200}) {constB=0}
  n37 = Add(A=@n32, B=VertexInterpolator(VS=NamedRerouteUsage()).PS)
  n38 = Min(A=Subtract(A=Multiply(A=@n37, B=@n34.PS), B=@n35) {constB=200}) {constB=0}
  n39 = DotProduct(A=Abs(None=Add(A=@n36, B=@n38)), B=Constant() {r=1})
  n40 = OneMinus(None=Saturate(None=Divide(A=@n39, B=@n35)))
  NamedRerouteDeclaration(None=Multiply(A=@n40, B=@n40)) {name=Bounds Mask, nodeColor=<rgba 0 0.8828 1 1>}
  n41 = VectorParameter() {bUseCustomPrimitiveData=true, parameterName=Color Day, primitiveDataIndex=3, defaultValue=<rgba 0.6726 0.2965 1 1>}
  n42 = VectorParameter() {bUseCustomPrimitiveData=true, parameterName=Color Evening, primitiveDataIndex=6, defaultValue=<rgba 0.7518 0.0417 1 1>}
  n43 = VectorParameter() {bUseCustomPrimitiveData=true, parameterName=Color Night, primitiveDataIndex=9, defaultValue=<rgba 0.7518 0.0417 1 1>}
  n44 = MaterialFunctionCall(
   Color_Morning=VectorParameter() {bUseCustomPrimitiveData=true, parameterName=Color Morning, defaultValue=<rgba 0.7518 0.0417 1 1>}.RGB
   Color_Day=@n41.RGB
   Color_Evening=@n42.RGB
   Color_Night=@n43.RGB
  ) {materialFunction=&Color_Time_of_Day}
  n45 = Add(A=DotProduct(A=NamedRerouteDeclaration(None=@n44.Result) {name=Color, nodeColor=<rgba 1 0.7969 0 1>}, B=Constant() {r=1})) {constB=-0.01}
  NamedRerouteDeclaration(None=Saturate(None=Ceil(None=@n45))) {name=Invisible Mask, nodeColor=<rgba 1 0.9375 0 1>}
  n46 = Transform(None=ActorPositionWS() {originType=CameraRelative}) {transformSourceType=TRANSFORMSOURCE_World, transformType=TRANSFORM_Local}
  n47 = Multiply(A=@n46) {constB=0.0099}
  MaterialExpressionMultiply_27 = Multiply(A=@n47, B=@n47)
  n48 = Convert(None=@MaterialExpressionMultiply_27) {
   nodeName=Break Float3
   convertInputs [{
     type=Vector3
     expressionInput {
      expression=&self:MaterialExpressionMultiply_27
      inputName=None
      mask=0
      maskA=0
      maskB=0
      maskG=0
      maskR=0
      outputIndex=0
     }
    }]
   convertOutputs [{type=Scalar}, {type=Scalar}, {type=Scalar}]
  }
  n49 = Add(A=Add(A=@n48, B=@n48), B=@n48)
  n50 = NamedRerouteUsage()
  n51 = Divide(
   A=Constant() {r=20}
   B=MaterialFunctionCall() {materialFunction=&ObjectScale}.Scale XYZ
  )
  n52 = Max(A=Subtract(A=@n46, B=Add(A=@n50, B=@n51) {constB=20})) {constB=0}
  n53 = NamedRerouteUsage()
  n54 = Min(A=Add(A=@n46, B=Add(A=@n53, B=@n51) {constB=20})) {constB=0}
  n55 = DotProduct(A=Abs(None=Add(A=@n52, B=@n54)), B=Constant() {r=1})
  n56 = Max(A=Floor(None=Saturate(None=@n49)), B=Ceil(None=Saturate(None=@n55)))
  NamedRerouteDeclaration(None=OneMinus(None=NamedRerouteDeclaration(None=@n56) {name=Outside Culling, nodeColor=<rgba 0 1 0.8828 1>})) {name=Inside Culling, nodeColor=<rgba 1 0.5625 0 1>}
  NamedRerouteUsage()
  NamedRerouteUsage()
  Reroute(None=@n28)
 ]
}
